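# Sound cue for the Blade Hazard's cut impact on a player, classed as a trap sound with cut attenuation.
# A Random node picks among six equally weighted slice waves.
# A Branch on IsLocal feeds that same Random node to all three outcomes, so local and non-local playback do not differ within the cue.
# It has high priority, disables virtualization and overrides concurrency to allow two simultaneous instances.
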SoundCue /BladeHazard/Assets/Sound/BladeHazard_Impact_Cut_Player_Cue
root=Branch_0
nodes [
 {id=Branch_0, children [Random_1,Random_1,Random_1], properties {boolParameterName=IsLocal}}
 {id=Random_1, children [WavePlayer_2,WavePlayer_3,WavePlayer_4,WavePlayer_5,WavePlayer_6,WavePlayer_7], properties {weights [1,1,1,1,1,1]}}
 {id=WavePlayer_2, properties {soundWaveAssetPtr=&/BladeHazard_Impact_Cut_Slice_01}}
 {id=WavePlayer_3, properties {soundWaveAssetPtr=&/BladeHazard_Impact_Cut_Slice_02}}
 {id=WavePlayer_4, properties {soundWaveAssetPtr=&/BladeHazard_Impact_Cut_Slice_03}}
 {id=WavePlayer_5, properties {soundWaveAssetPtr=&/BladeHazard_Impact_Cut_Slice_04}}
 {id=WavePlayer_6, properties {soundWaveAssetPtr=&/BladeHazard_Impact_Cut_Slice_05}}
 {id=WavePlayer_7, properties {soundWaveAssetPtr=&/BladeHazard_Impact_Cut_Slice_06}}
]
properties {
 attenuationSettings=&/BladeHazard_Cut_Att
 bOverrideConcurrency=true
 maxDistance=3000
 priority=0.9
 soundClassObject=&Fort_Traps
 virtualizationMode=Disabled
 volumeMultiplier=0.7
 concurrencyOverrides {maxCount=2}
 concurrencySet [/Game/Sounds/Concurrency_Rules/Athena_Prop]
}
