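# Sound cue for a blade hazard's cutting impact, named for vehicle hits and routed through the traps sound class.
# A Branch on the IsLocal parameter selects one of three Mixers, each fed by an equally weighted Random over six identically named slice samples.
# All three paths share mixer volume and weighting.
# The cue applies cut attenuation, disables virtualization and overrides concurrency with a low instance cap.
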SoundCue /BladeHazard/Assets/Sound/BladeHazard_Impact_Cut_Vehicle_Cue
root=Branch_0
nodes [
 {id=Branch_0, children [Mixer_0,Mixer_1,Mixer_2], properties {boolParameterName=IsLocal}}
 {id=Mixer_0, children [Random_1], properties {inputVolume [1]}}
 {id=Mixer_1, children [Random_0], properties {inputVolume [1]}}
 {id=Mixer_2, children [Random_2], properties {inputVolume [1]}}
 {
  id=Random_0
  children [WavePlayer_0, WavePlayer_1, WavePlayer_8, WavePlayer_9, WavePlayer_10, WavePlayer_11]
  properties {weights [1,1,1,1,1,1]}
 }
 {id=Random_1, children [WavePlayer_2,WavePlayer_3,WavePlayer_4,WavePlayer_5,WavePlayer_6,WavePlayer_7], properties {weights [1,1,1,1,1,1]}}
 {
  id=Random_2
  children [WavePlayer_12, WavePlayer_13, WavePlayer_14, WavePlayer_15, WavePlayer_16, WavePlayer_17]
  properties {weights [1,1,1,1,1,1]}
 }
 {id=WavePlayer_0, properties {soundWaveAssetPtr=&/BladeHazard_Impact_Cut_Slice_01}}
 {id=WavePlayer_1, properties {soundWaveAssetPtr=&/BladeHazard_Impact_Cut_Slice_02}}
 {id=WavePlayer_10, properties {soundWaveAssetPtr=&/BladeHazard_Impact_Cut_Slice_05}}
 {id=WavePlayer_11, properties {soundWaveAssetPtr=&/BladeHazard_Impact_Cut_Slice_06}}
 {id=WavePlayer_12, properties {soundWaveAssetPtr=&BladeHazard_Impact_Cut_Slice_01}}
 {id=WavePlayer_13, properties {soundWaveAssetPtr=&BladeHazard_Impact_Cut_Slice_02}}
 {id=WavePlayer_14, properties {soundWaveAssetPtr=&/BladeHazard_Impact_Cut_Slice_03}}
 {id=WavePlayer_15, properties {soundWaveAssetPtr=&/BladeHazard_Impact_Cut_Slice_04}}
 {id=WavePlayer_16, properties {soundWaveAssetPtr=&BladeHazard_Impact_Cut_Slice_05}}
 {id=WavePlayer_17, properties {soundWaveAssetPtr=&BladeHazard_Impact_Cut_Slice_06}}
 {id=WavePlayer_2, properties {soundWaveAssetPtr=&BladeHazard_Impact_Cut_Slice_01}}
 {id=WavePlayer_3, properties {soundWaveAssetPtr=&BladeHazard_Impact_Cut_Slice_02}}
 {id=WavePlayer_4, properties {soundWaveAssetPtr=&BladeHazard_Impact_Cut_Slice_03}}
 {id=WavePlayer_5, properties {soundWaveAssetPtr=&BladeHazard_Impact_Cut_Slice_04}}
 {id=WavePlayer_6, properties {soundWaveAssetPtr=&BladeHazard_Impact_Cut_Slice_05}}
 {id=WavePlayer_7, properties {soundWaveAssetPtr=&BladeHazard_Impact_Cut_Slice_06}}
 {id=WavePlayer_8, properties {soundWaveAssetPtr=&BladeHazard_Impact_Cut_Slice_03}}
 {id=WavePlayer_9, properties {soundWaveAssetPtr=&BladeHazard_Impact_Cut_Slice_04}}
]
properties {
 attenuationSettings=&/BladeHazard_Cut_Att
 bOverrideConcurrency=true
 maxDistance=3000
 priority=0.9
 soundClassObject=&Fort_Traps
 virtualizationMode=Disabled
 volumeMultiplier=0.7
 concurrencyOverrides {maxCount=2}
 concurrencySet [/Game/Sounds/Concurrency_Rules/Athena_Prop]
}
