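# Looping Sound Cue for the spinning blade hazard, classed as a trap sound and played with its own attenuation.
# A Random node picks one of two looping spin waves with equal weight and feeds a Modulator that slightly randomizes pitch upward.
# An Enveloper at the root fades the volume in.
# Concurrency is overridden to cap simultaneous instances, with a short release when voices are stolen.
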SoundCue /BladeHazard/Assets/Sound/BladeHazard_Loop_Cue
root=Enveloper_1
nodes [
 {
  id=Enveloper_1
  children [Modulator_0]
  properties {pitchCurve {editorCurveData {keys [{time=0}]}}, volumeCurve {editorCurveData {keys [<at 0 0>,<at 1 1>]}}}
 }
 {id=Modulator_0, children [Random_0], properties {pitchMax=1.2, pitchMin=1, volumeMax=1, volumeMin=1}}
 {id=Random_0, children [WavePlayer_0,WavePlayer_1], properties {bShouldExcludeFromBranchCulling=true, weights [1,1]}}
 {id=WavePlayer_0, properties {bLooping=true, soundWaveAssetPtr=&/BladeHazard_Spin_Loop_01}}
 {id=WavePlayer_1, properties {bLooping=true, soundWaveAssetPtr=&/BladeHazard_Spin_Loop_02}}
]
properties {
 attenuationSettings=&/BladeHazard_Loop_Att
 bOverrideConcurrency=true
 maxDistance=3000
 soundClassObject=&Fort_Traps
 volumeMultiplier=0.7
 concurrencyOverrides {bVolumeScaleCanRelease=true, maxCount=3, voiceStealReleaseTime=0.15, volumeScaleAttackTime=0.5}
 concurrencySet [/Game/Sounds/Concurrency_Rules/Athena_Prop]
}
